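; PCG Biome Core filter graph that applies noise to point density.
; Input data is converted with To Point, then passed through Attribute Noise, which runs on its default settings apart from Mode.
; The noise mode is exposed as the graph parameter AttributeNoiseMode, so the graph's user chooses how the noise combines with existing values.
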(graph BiomeFilter_Density_Noise
  (params (AttributeNoiseMode :type (Enum /Script/PCG.EPCGAttributeNoiseMode)))
  (bind ToPoint_1 (To%20Point))
  (bind AttributeNoise_0 (Attribute%20Noise :Mode-param AttributeNoiseMode))
  (connect ToPoint_1 Out AttributeNoise_0 In)
  (connect DefaultInputNode In ToPoint_1 In)
  (output Out AttributeNoise_0))
