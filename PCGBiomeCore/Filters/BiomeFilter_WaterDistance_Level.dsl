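; PCG biome filter graph that tags points with their distance to water, using a height threshold as the water level.
; It splits input points by comparing Position.z with the WaterLevel parameter.
; The inside-filter points act as Distance sources and the outside-filter points as targets, so no water geometry is needed.
; Output source points carry a center-to-center WaterDistance attribute, with search range set by WaterLevelMaxDistance.
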(graph BiomeFilter_WaterDistance_Level
  (params (WaterLevel :type Double :default -1e+02) (WaterLevelMaxDistance :type Double :default 2e+04))
  (bind AttributeFilter_0 (Filter%20Attribute%20Elements :AttributeTypes (struct :DoubleValue 1e+02) :TargetAttribute $Position.z :ThresholdAttribute WaterLevel :Filter WaterLevel))
  (bind Distance_1 (Distance :OutputAttribute WaterDistance :SourceShape Center :TargetShape Center :MaximumDistance-param WaterLevelMaxDistance))
  (connect DefaultInputNode In AttributeFilter_0 In)
  (connect AttributeFilter_0 InsideFilter Distance_1 Source)
  (connect AttributeFilter_0 OutsideFilter Distance_1 Target)
  (output Out Distance_1))
